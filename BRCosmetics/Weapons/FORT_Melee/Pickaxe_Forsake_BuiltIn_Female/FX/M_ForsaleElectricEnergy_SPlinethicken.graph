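# Masked, two-sided unlit material for electric lightning on the female Forsake pickaxe, built for spline-thicken geometry.
# Panning noise distorts band-contrasted UVs that blend end and middle lightning textures, with vertex red offsetting time per helix.
# The masked texture drives opacity and an inner/outer color lerp, boosted to HDR with eye-adaptation control.
# World position offset combines SplineThicken, noise jitter and banded tangent spiral push, with noise also varying width.
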Material /BRCosmetics/Weapons/FORT_Melee/Pickaxe_Forsake_BuiltIn_Female/FX/M_ForsaleElectricEnergy_SPlinethicken
graph {
 properties {
  bUsedWithMeshDeformer=true
  bUsedWithNiagaraMeshParticles=true
  bUsedWithNiagaraRibbons=true
  bUsedWithNiagaraSprites=true
  bUsedWithSkeletalMesh=true
  blendMode=BLEND_Masked
  pixelDepthOffsetMode=PDOM_Legacy
  shadingModel=MSM_Unlit
  twoSided=true
  usedShadingModels=Unlit
 }
 nodes [
  n1 = TextureCoordinate()
  n2 = AppendVector(
   A=ScalarParameter() {defaultValue=1, group=uv Distortion, parameterName=Distort Tile X}
   B=ScalarParameter() {defaultValue=6, group=uv Distortion, parameterName=Distort Tile Y}
  )
  n3 = Add(
   A=Time()
   B=Multiply(A=VertexColor().R, B=ScalarParameter() {defaultValue=0.5, group=UV, parameterName=time offset for one helix}) {constB=0.5}
  )
  n4 = ScalarParameter() {group=uv Distortion, parameterName=Distort SPeed X}
  n5 = ScalarParameter() {group=uv Distortion, parameterName=Distort SPeed Y}
  n6 = TextureSampleParameter2D(UVs=Panner(Coordinate=Multiply(A=@n1, B=@n2), Time=@n3, Speed=AppendVector(A=@n4, B=@n5)) {speedY=0.5}) {
   group=uv Distortion
   parameterName=Distort Noise Texture
   samplerType=SAMPLERTYPE_LinearColor
   texture=&HighContrastNoise
  }
  n7 = ScalarParameter() {defaultValue=0.5, group=uv Distortion, parameterName=Distort band with}
  n8 = ScalarParameter() {defaultValue=0.5, group=uv Distortion, parameterName=Distort band sharpness}
  n9 = Multiply(
   A=Subtract(A=ComponentMask(None=@n6.RGB) {g=true, r=true}) {constB=0.5}
   B=Multiply(
    A=ScalarParameter() {defaultValue=0.1, group=uv Distortion, parameterName=Distort Str}
    B=MaterialFunctionCall!(Width=@n7, Sharpness=@n8, Offset, Direction Switch, Compare, Input Coordinates) {materialFunction=/Engine/Functions/Engine_MaterialFunctions03/Procedurals/GeneratedBand}.Result
   )
  ) {constB=2}
  n10 = TextureCoordinate()
  n11 = VectorParameter() {group=UV, parameterName=Texture Options, channelNames {a=Speed Y, b=Speed X, g=Tile Y, r=Tile X}, defaultValue=<rgba 1 1 0 0.25>}
  n12 = Panner(
   Coordinate=Multiply(A=Add(A=@n9, B=@n10), B=AppendVector(A=@n11.Tile X, B=@n11.Tile X))
   Time=@n3
   Speed=AppendVector(A=@n11.Speed X, B=@n11.Speed X)
  ) {speedY=0.35}
  n13 = LinearInterpolate(
   A=@n12
   B=MaterialFunctionCall(XY=@n12) {materialFunction=&Swizzle}.YX
   Alpha=ScalarParameter() {defaultValue=1, group=UV, parameterName=Swizzzle Uvs}
  )
  n14 = MaterialFunctionCall(Float2=@n13) {materialFunction=&BreakOutFloat2Components}
  n15 = Round(None=MaterialFunctionCall(Float2=TextureCoordinate()) {materialFunction=&BreakOutFloat2Components}.G)
  n16 = LinearInterpolate(
   A=ScalarParameter() {defaultValue=5, group=UV, parameterName=UV Contrast Top}
   B=ScalarParameter() {defaultValue=5, group=UV, parameterName=UV Contrast bottom}
   Alpha=@n15
  )
  n17 = ScalarParameter() {defaultValue=1, group=UV, parameterName=UV Contrast middle}
  n18 = LinearInterpolate(
   A=MaterialFunctionCall(
    Width=ScalarParameter() {defaultValue=0.5, group=UV, parameterName=UV Band Width Top}
    Sharpness=ScalarParameter() {defaultValue=0.5, group=UV, parameterName=UV Band Sharpness Top}
   ) {materialFunction=&GeneratedBand}.Result
   B=MaterialFunctionCall(
    Width=ScalarParameter() {defaultValue=0.5, group=UV, parameterName=UV Band Width Bottom}
    Sharpness=ScalarParameter() {defaultValue=0.5, group=UV, parameterName=UV Band Sharpness Bottom}
   ) {materialFunction=&GeneratedBand}.Result
   Alpha=@n15
  )
  n19 = MaterialFunctionCall(
   X=@n14.R
   Y=MaterialFunctionCall(In=@n14.G, Contrast=LinearInterpolate(A=@n16, B=@n17, Alpha=@n18)) {materialFunction=&CheapContrast}.Result
  ) {materialFunction=&MakeFloat2}
  n20 = TextureSampleParameter2D(UVs=@n19.Result) {parameterName=Texture ends, texture=&/T_ForsakePickLightning_1_basecolor}
  n21 = ScalarParameter() {defaultValue=0.45, parameterName="Texture  Band width"}
  n22 = ScalarParameter() {defaultValue=0.75, parameterName=Textuyre Band Sharpness}
  n23 = Multiply(
   A=LinearInterpolate(
    A=@n20.RGBA
    B=TextureSampleParameter2D(UVs=@n19.Result) {parameterName=Texture, texture=&/T_ForsakePickLightning_basecolor}.RGBA
    Alpha=MaterialFunctionCall(Width=@n21, Sharpness=@n22) {materialFunction=&GeneratedBand}.Result
   )
   B=ChannelMaskParameter(None=TextureSampleParameter2D() {parameterName=Mask Texture, texture=/Engine/EngineResources/WhiteSquareTexture}.RGBA) {parameterName=Mask Channel}.R
  )
  n24 = DotProduct(A=ComponentMask(None=@n23) {b=true, g=true, r=true}, B=Constant() {r=1})
  Add(A=@n24, B=ScalarParameter() {defaultValue=0.333, parameterName=Opacity Mask Offset}) {constA=1}
  n25 = Multiply(A=@n3, B=ScalarParameter() {defaultValue=0.25, group=Wpo, parameterName=WPO SPeed})
  n26 = Panner(Time=@n25) {speedY=0.5}
  n27 = TextureSampleParameter2D(UVs=AppendVector(
    A=@n25
    B=MaterialFunctionCall(UV Channel=@n26) {materialFunction=&LinearGradient}.VGradient
   )) {parameterName=WPO Noise Texture, samplerType=SAMPLERTYPE_LinearColor, texture=&HighContrastNoise}
  n28 = ScalarParameter() {defaultValue=0.5, group=Wpo, parameterName=wpo band with}
  n29 = ScalarParameter() {defaultValue=0.5, group=Wpo, parameterName=wpo band sharpness}
  n30 = Multiply(
   A=Subtract(A=@n27.RGB) {constB=0.5}
   B=Multiply(
    A=ScalarParameter() {defaultValue=5, group=Wpo, parameterName=Wpo Str}
    B=MaterialFunctionCall(Width=@n28, Sharpness=@n29) {materialFunction=&GeneratedBand}.Result
   )
  ) {constB=2}
  n31 = ScalarParameter() {group=spiral WPO, parameterName=SPiral Push bottom end}
  n32 = ScalarParameter() {group=spiral WPO, parameterName=SPiral Push top end}
  n33 = Round(None=MaterialFunctionCall() {materialFunction=&LinearGradient}.VGradient)
  n34 = ScalarParameter() {defaultValue=0.5, group=spiral WPO, parameterName=spiral Band Width Bottom}
  n35 = ScalarParameter() {defaultValue=0.5, group=spiral WPO, parameterName=spiral Band Sharpness Bottom}
  n36 = Multiply(
   A=VertexTangentWS()
   B=LinearInterpolate(
    A=LinearInterpolate(A=@n31, B=@n32, Alpha=@n33)
    B=ScalarParameter() {group=spiral WPO, parameterName=SPiral Push Middle}
    Alpha=MaterialFunctionCall(Width=@n34, Sharpness=@n35) {materialFunction=&GeneratedBand}.Result
   )
  )
  n37 = Add(A=Add(A=WorldPosition().XYZ, B=@n30), B=@n36)
  n38 = DotProduct(A=@n30, B=Constant() {r=1})
  n39 = LinearInterpolate(
   A=ScalarParameter() {defaultValue=5, parameterName=Width ends}
   B=LinearInterpolate(
    A=ScalarParameter() {defaultValue=5, group=Spline Thicken, parameterName=Width middle Low}
    B=ScalarParameter() {defaultValue=5, group=Spline Thicken, parameterName=Width middle High}
    Alpha=Saturate(None=@n38)
   )
   Alpha=MaterialFunctionCall(
    Width=ScalarParameter() {defaultValue=0.5, group=Spline Thicken, parameterName=Band Width}
    Sharpness=ScalarParameter() {defaultValue=0.9, group=Spline Thicken, parameterName=Band Sharpness}
   ) {materialFunction=&GeneratedBand}.Result
  )
  Add(
   A=Add(
    A=MaterialFunctionCall(WidthBase=@n39, WidthTip=@n39, WorldPosition=@n37) {materialFunction=&SplineThicken}.WorldPositionOffset
    B=@n30
   )
   B=@n36
  )
  n40 = ScalarParameter() {defaultValue=1, parameterName=PdoTexture Tile}
  LinearInterpolate(
   A=ScalarParameter() {parameterName=PDO Str low}
   B=ScalarParameter() {parameterName=PDO Str high}
   Alpha=TextureSample(UVs=Multiply(A=@n13, B=@n40)) {samplerType=SAMPLERTYPE_LinearGrayscale, texture=/Engine/EngineMaterials/Good64x64TilingNoiseHighFreq}.R
  )
  LinearInterpolate()
  n41 = VectorParameter() {parameterName=Color Lerp Options, channelNames {a="", b=add, g=Mul, r=power}, defaultValue=<rgba 1 1 0 0>}
  n42 = Multiply(
   A=Power(Base=ChannelMaskParameter(None=@n23) {maskChannel=Alpha, parameterName=Color Lerp Channel, defaultValue=<rgba 0 0 0 1>}.R, Exp=@n41.power)
   B=@n41.Mul
  )
  n43 = LinearInterpolate(
   A=VectorParameter() {group=color, parameterName=Outer Color, defaultValue=<rgba 52 0 60 0>}.RGB
   B=VectorParameter() {group=color, parameterName=Inner COlor, defaultValue=<rgba 1 1 1 0>}.RGB
   Alpha=Clamp(None=Add(A=@n42, B=@n41.add))
  )
  MaterialExpressionMultiply_4 = Multiply(
   A=MaterialFunctionCall(
    LDR Input=Saturate(None=@n43)
    HDR Dynamic Range=ScalarParameter() {defaultValue=100, group=Derive HDR from LDR, parameterName="HDR Dynamic Range (S)"}
    HDR Tint=@n43
    HDR Intensity Multiplier=ScalarParameter() {defaultValue=1, group=Derive HDR from LDR, parameterName="HDR Intensity Multiplier (S)"}
    Derive HDR Power=ScalarParameter() {defaultValue=1, group=Derive HDR from LDR, parameterName="Derive HDR Power (S)"}
    Desaturate Luminance=ScalarParameter() {defaultValue=1, group=Derive HDR from LDR, parameterName="Desaturate Luminance (S)"}
   ) {materialFunction=&DeriveHDRfromLDR}.RemappedRange
   B=LinearInterpolate(B=ComponentMask(None=@n23) {b=true}, Alpha=ScalarParameter() {defaultValue=1, parameterName=use blue channel}) {constA=1}
  )
  n44 = VectorParameter() {
   parameterName=Use Eye Adaption Exposure Control for Emissives TOD
   channelNames {a=Night, b=evening, g=Day, r=Morn}
   defaultValue=<rgba 0 0 0 2.5>
  }
  MaterialExpressionMaterialFunctionCall_9 = MaterialFunctionCall(Value_Morning=@n44.Morn, Value_Day=@n44.Morn, Value_Evening=@n44.Morn, Value_Night=@n44.Morn) {materialFunction=&Scalar_Time_of_Day}
  MaterialExpressionEyeAdaptation_1 = EyeAdaptation()
  n45 = Custom!(
   LightValue=@MaterialExpressionMultiply_4
   Alpha=@MaterialExpressionMaterialFunctionCall_9.Result
   Adaptation=@MaterialExpressionEyeAdaptation_1.EyeAdaptation
  ) {
   code="float LerpLogScale = -Alpha * log(Adaptation);\r\nfloat Scale = exp(LerpLogScale);\r\nreturn LightValue * Scale;\r\n\r\n\r\n"
   inputs [
    {
     inputName=LightValue
     input {expression=&self:MaterialExpressionMultiply_4}
    }
    {
     inputName=Alpha
     input {expression=&self:MaterialExpressionMaterialFunctionCall_9}
    }
    {
     inputName=Adaptation
     input {expression=&self:MaterialExpressionEyeAdaptation_1}
    }
   ]
  }
  MaterialFunctionCall(Emissive=@n45) {materialFunction=&MF_FX_EmissiveAdjustments}
  Reroute(None=@n24)
 ]
}
