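# Niagara swing effect for the female built-in Forsake pickaxe, attached to its skeletal mesh and tinted pink or, in an alt state, blue.
# Lights pins four persistent glow sprites to sockets, one casting light, with time-of-day alpha.
# Sprites and Thrust001 spawn short-lived sprites and meshes angled outward around an offset socket, scaled by System Size Lerp.
# Trail001 draws a fading ribbon from that point, recording distance travelled for UVs.
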NiagaraSystem /BRCosmetics/Weapons/FORT_Melee/Pickaxe_Forsake_BuiltIn_Female/FX/NS_ForsakeBuiltin_Swing
dependencies {
 usedDataInterfaces [
  &NiagaraDataInterfaceCurve
  &NiagaraDataInterfaceVector2DCurve
  &NiagaraDataInterfaceArrayInt32
  &NiagaraDataInterfaceArrayFloat3
  &NiagaraDataInterfaceArrayFloat
  &NiagaraDataInterfaceVectorCurve
 ]
 usedDynamicInputs [&SetBoolByFloatComparison]
 usedModules [
  &self:SystemSpawnScript.SystemScriptSource.SystemScriptGraph.NiagaraNodeAssignment_0.SetVariables_#1
  &self:SystemSpawnScript.SystemScriptSource.SystemScriptGraph.NiagaraNodeAssignment_1.SetVariables_#2
  &SystemState
 ]
}
emitters {
 Lights {
  emitterUpdateScript [
   &EmitterState()
   &SpawnBurst_Instantaneous(Age:NiagaraFloat=$Emitter.Age, Spawn Count=4, Use Spawn Probability=false, UseLoopCountLimit=false)
  ]
  particleSpawnScript [&InitializeParticle(
    Color=&MultiplyLinearColor(
     Linear Color=&MakeCustomLinearColorFromBool(Bool=$System.IsAltBool, True Color=<rgba 0 0 1 1>, False Color=<rgba 1 0 0.7999 1>)
     Scale Factor=<1 1 1 1>
    )
    Lifetime=5
    Sprite Size=<22 33>
    Sprite Size Mode=::NewEnumerator4
    UsePositionOffset=false
    Write Material Random=true
   )]
  particleUpdateScript [
   &ParticleState(
    Kill Particles When Lifetime Has Elapsed=false
    Let Infinitely Lived Particles Die When Emitter Deactivates=true
    Loop Particles Lifetime=true
   )
   &SkeletalMeshLocation(
    Bone / Socket Index=&SelectIntFromArray(
     Direct Array Index=&ReturnExecIndex()
     Int Selection Array=DataInterface() {propertyValues {IntData [0,1,2,2]}}
    )
    Bone / Socket Sampling Mode=::NewEnumerator5
    MaskSpawnGroup=false
    Mesh Sampling Type=::NewEnumerator6
    Sampled Position Offset=&SelectVectorFromArray(
     Direct Array Index=&ReturnExecIndex()
     Vector Selection Array=DataInterface() {propertyValues {FloatData [<0 0 0>, <0 0 0>, <0 0 0>, <13 0 0>]}}
    )
    Skeletal Mesh=$User.SkeletalMesh
   )
   &FN_Light_Attributes(
    Color=&HueShiftLinearColor(Hue=0, Saturation=0.5, Value=1, Linear Color=$Particles.Color)
    Create Color Attribute=true
    Enable Particle Light Enable=true
    Light Enabled:NiagaraBool=&SetBoolByIntComparison(A=&ModuloInt(A=&ReturnExecIndex(), B=4), B=0, Comparison Type=::NewEnumerator0)
    Scale Color=3
   )
   /SetVariables(Particles.Sprite COlor:LinearColor=&MultiplyLinearColor(Linear Color=$Particles.Color, Scale Factor=&MakeVector4(X=1, Y=1, Z=1, W=0.5))) {isSetParameters=true}
   /SetVariables(Particles.Velocity=&ReturnMeshOrientationAxis(Mesh Orientation Axis=::XAxis, Mesh Orientation Quat=$Particles.SkeletalMeshLocation.SampledRotation)) {isSetParameters=true}
   &CameraOffset(Camera Offset Amount=2)
   &ScaleColor(
    Scale Alpha=&FNTimeOfDayScalar(Night Scale=1, Morning Scale=0.75, Evening Scale=0.65)
    ScaleA=true
    ScaleRGB=true
   )
  ]
  rendererClasses [&NiagaraSpriteRendererProperties, &NiagaraLightRendererProperties]
  renderers [
   {rendererClass=&NiagaraSpriteRendererProperties}
   {rendererClass=&NiagaraLightRendererProperties, rendererIndex=1}
  ]
 }
 Sprites {
  emitterUpdateScript [
   &EmitterState()
   &SpawnRate(SpawnRate=&Multiply_Float(A=8, B=4), Use Spawn Probability=false)
  ]
  particleSpawnScript [
   &InitializeParticle(
    Color=&MakeCustomLinearColorFromBool(Bool=$System.IsAltBool, True Color=<rgba 0 0 1 1>, False Color=<rgba 1 0 0.7999 1>)
    Lifetime Max=0.2
    Lifetime Min=0.3
    Lifetime Mode=::NewEnumerator1
    Mesh Scale Max=<0.25 1 0.5>
    Mesh Scale Min=<0.35 1 1>
    Mesh Scale Mode=::NewEnumerator2
    Sprite Size Max=<5 10>
    Sprite Size Min=<10 8>
    Sprite Size Mode=::NewEnumerator2
    UsePositionOffset=false
    Write Material Random=true
   )
   &InitialMeshOrientation(
    Mesh Orientation Mode=::NewEnumerator7
    Rotation=&Add_Vector(
     A=<0 0.25 0>
     B=&Multiply_VectorByFloat(
      Vector=&RandomRangeVector(
       Minimum=&VectorFromFloat(Value=0.12)
       Maximum=&VectorFromFloat(Value=-0.12)
      )
      Float=&FloatFromCurve(
       FloatCurve=DataInterface() {propertyValues {Curve {keys [{tangentMode=RCTM_User, time=0, value=0}, {tangentMode=RCTM_User, time=1, value=1}]}}}
       CurveIndex=$System.Size Lerp
      )
     )
    )
    Use Rotation Vector=true
   )
  ]
  particleUpdateScript [
   &ParticleState()
   &SkeletalMeshLocation(
    Bone / Socket Index=3
    Bone Sampling Mode=::NewEnumerator4
    MaskSpawnGroup=false
    Orientation Sampling=::NewEnumerator1
    Sampled Position Offset=&Add_Vector(
     A=<30 0 0>
     B=/Niagara/DynamicInputs/Vectors/RotateVector(
      Vector To Rotate:Vector3f=&RandomRangeVector(Minimum=<10 -1 0>, Maximum=<10 1 0>)
      Yaw:NiagaraFloat=0
      Pitch:NiagaraFloat=&Add_Float(
       A=45
       B=&Multiply_Float(A=&MakeFloatFromInt(Int=&ReturnParticleIDAttributes()), B=90)
      )
      Roll:NiagaraFloat=0
      Rotation Rate:NiagaraFloat=1
     )
    )
    Skeletal Mesh=$User.SkeletalMesh
   )
   &ScaleColor(
    Scale Alpha=&FloatFromCurve(FloatCurve=DataInterface() {propertyValues {Curve {keys [<at 0.75 1>,<at 1 0>]}}})
    Scale RGB=&VectorFromFloat(Value=5)
    ScaleA=true
    ScaleRGB=true
   )
   &UpdateMeshOrientation(
    Facing Direction=&Subtract_Vector(
     A=Linked() {linkedVariable {name=Particles.Position, type=NiagaraPosition}}
     B=&Add_Vector(
      A=Linked() {linkedVariable {name=Output.SkeletalMeshLocation.SampledPositionWithoutOffset, type=NiagaraPosition}}
      B=&Multiply_Vector(
       A=&Multiply_VectorByFloat(
        Vector=&ReturnMeshOrientationAxis(Mesh Orientation Axis=::XAxis, Mesh Orientation Quat=$Particles.SkeletalMeshLocation.SampledRotation)
        Float=30
       )
       B=<1 1 1>
      )
     )
    )
    Facing Interpolation=false
    Orientation Method=::NewEnumerator1
    Override Quat=false
    TransformFacing=false
    TransformUp=false
   )
   &AlignSpriteToMeshOrientation(Mesh Orientation Relative Sprite Alignment Vector=<0 0 1>)
   &ScaleSpriteSize(
    Scale Factor=&Lerp_Vector2(
     A=&Vector2DFromCurve(
      Vector2Curve=DataInterface() {propertyValues {
        XCurve {keys [{tangentMode=RCTM_User, time=0, value=0}, <at 0.5 1>, {tangentMode=RCTM_User, time=1, value=1}]}
        YCurve {keys [
          {interpMode=RCIM_Cubic, tangentMode=RCTM_User, time=0, value=0}
          {
           arriveTangentWeight=0.2357
           interpMode=RCIM_Cubic
           leaveTangent=-6.8837
           leaveTangentWeight=0.2357
           tangentMode=RCTM_Break
           time=0.5
           value=1
          }
          {
           arriveTangentWeight=0.4472
           interpMode=RCIM_Cubic
           leaveTangentWeight=0.4472
           tangentMode=RCTM_User
           tangentWeightMode=RCTWM_WeightedBoth
           time=1
           value=0
          }
         ]}
       }}
      ScaleCurve=<0.1 0.5>
     )
     B=&Vector2DFromCurve(
      Vector2Curve=DataInterface() {propertyValues {
        XCurve {keys [{tangentMode=RCTM_User, time=0, value=0}, <at 0.5 1>, {tangentMode=RCTM_User, time=1, value=1}]}
        YCurve {keys [
          {interpMode=RCIM_Cubic, tangentMode=RCTM_User, time=0, value=0}
          {
           arriveTangentWeight=0.2357
           interpMode=RCIM_Cubic
           leaveTangent=0.7831
           leaveTangentWeight=0.2357
           tangentMode=RCTM_Break
           time=0.2
           value=1.1
          }
          {
           arriveTangentWeight=0.2347
           interpMode=RCIM_Cubic
           leaveTangentWeight=0.2347
           tangentMode=RCTM_User
           tangentWeightMode=RCTWM_WeightedBoth
           time=1
           value=0
          }
         ]}
       }}
      ScaleCurve=<0.5 1>
     )
     Alpha=&FloatFromCurve(
      FloatCurve=DataInterface() {propertyValues {Curve {keys [{tangentMode=RCTM_User, time=0, value=0}, {tangentMode=RCTM_User, time=1, value=1}]}}}
      CurveIndex=$System.Size Lerp
     )
    )
    Scale Sprite Size Mode=::NewEnumerator2
   )
   &ScaleSpriteSize(
    Scale Sprite Size Mode=::NewEnumerator0
    Uniform Scale Factor=&FloatFromCurve(FloatCurve=DataInterface() {propertyValues {Curve {keys [<at 1.4 0.4>,<at 100 1.25>]}}}, CurveIndex=$System.Size Lerp)
   ) {module=ScaleSpriteSize001}
   /SetVariables(Particles.OffsetPosition:NiagaraPosition=/Niagara/DynamicInputs/Lerp/Lerp_Position(
     A:NiagaraPosition=$Particles.Position
     B:NiagaraPosition=&Add_Vector(
      A=Linked() {linkedVariable {name=Particles.Position, type=NiagaraPosition}}
      B=&Multiply_VectorByFloat(
       Vector=&ReturnMeshOrientationAxis(Mesh Orientation Axis=::ZAxis, Mesh Orientation Quat=$Particles.MeshOrientation)
       Float=&RandomRangeFloat(Minimum=0, Maximum=25)
      )
     )
     Alpha:NiagaraFloat=&FloatFromCurve(
      FloatCurve=DataInterface() {propertyValues {Curve {keys [{arriveTangent=2.3442, interpMode=RCIM_Cubic, leaveTangent=2.3442, tangentMode=RCTM_User, time=0, value=0}, <at 1 1 RCIM_Cubic>]}}}
      Scale Curve=&Lerp_Float(
       A=0.35
       B=0.5
       Alpha=&FloatFromCurve(
        FloatCurve=DataInterface() {propertyValues {Curve {keys [{arriveTangent=2.3442, interpMode=RCIM_Cubic, leaveTangent=2.3442, tangentMode=RCTM_User, time=0, value=0}, <at 1 1 RCIM_Cubic>]}}}
        CurveIndex=$System.Size Lerp
       )
      )
     )
    )) {isSetParameters=true}
  ]
  rendererClasses [&NiagaraSpriteRendererProperties]
  renderers [{rendererClass=&NiagaraSpriteRendererProperties}]
 }
 Thrust001 {
  emitterUpdateScript [
   &EmitterState()
   &SpawnRate(SpawnRate=&Multiply_Float(A=35, B=4), Use Spawn Probability=false)
  ]
  particleSpawnScript [
   &InitializeParticle(
    Color=&MakeCustomLinearColorFromBool(Bool=$System.IsAltBool, True Color=<rgba 0 0 1 1>, False Color=<rgba 1 0 0.7999 1>)
    Lifetime=0.2
    Mesh Scale=&Multiply_VectorByFloat(Vector=&RandomRangeVector(Minimum=<0.5 0.5 0.5>, Maximum=<1 0.75 0.75>), Float=0.1)
    Mesh Scale Mode=::NewEnumerator4
    UsePositionOffset=false
    Write Material Random=true
   )
   &InitialMeshOrientation(
    Mesh Orientation Mode=::NewEnumerator7
    Rotation=&Add_Vector(A=<0 0 0>, B=&RandomRangeVector(Minimum=<0 0 0>, Maximum=<1 0 0>))
    Use Rotation Vector=true
   )
  ]
  particleUpdateScript [
   &ParticleState()
   &SkeletalMeshLocation(
    Bone / Socket Index=3
    Bone Sampling Mode=::NewEnumerator4
    MaskSpawnGroup=false
    Orientation Sampling=::NewEnumerator1
    Sampled Position Offset=&Add_Vector(
     A=<30 0 0>
     B=&Add_Vector(
      A=&RotateVector(
       Vector To Rotate:Vector3f=<10 0 0>
       Yaw:NiagaraFloat=0
       Pitch:NiagaraFloat=&SelectFloatFromArray(
        Direct Array Index=&ReturnParticleIDAttributes()
        Direct Array Mode=::NewEnumerator1
        Float Selection Array=DataInterface() {propertyValues {FloatData [44,-134,134,-44]}}
       )
       Roll:NiagaraFloat=0
       Rotation Rate:NiagaraFloat=1
      )
      B=&Multiply_VectorByFloat(
       Vector=&MakeVector(
        X=0
        Y=&Lerp_Float(
         A=&FloatFromCurve(FloatCurve=DataInterface() {propertyValues {Curve {keys [<at 0 -1>,<at 1 1>]}}})
         B=&FloatFromCurve(
          FloatCurve=DataInterface() {propertyValues {Curve {keys [
              {interpMode=RCIM_Cubic, leaveTangentWeight=0.0333, tangentMode=RCTM_User, time=0, value=-1}
              {arriveTangentWeight=0.0745, interpMode=RCIM_Cubic, leaveTangentWeight=0.2981, tangentMode=RCTM_User, time=0.15, value=0}
              {arriveTangentWeight=0.2981, interpMode=RCIM_Cubic, leaveTangentWeight=0.0745, tangentMode=RCTM_User, time=0.85, value=0}
              {arriveTangentWeight=0.0333, interpMode=RCIM_Cubic, tangentMode=RCTM_User, time=1, value=1}
             ]}}}
          CurveIndex=&RandomRangeFloat(Minimum=0, Maximum=1)
         )
         Alpha=0.5
        )
        Z=0
       )
       Float=1
      )
     )
    )
    Skeletal Mesh=$User.SkeletalMesh
    Use Sampled Mesh Attribute as Ribbon ID=false
   )
   &ScaleMeshSize(Scale Factor=&MakeVector(
     X=&FloatFromCurve(
      FloatCurve=DataInterface() {propertyValues {Curve {keys [{tangentMode=RCTM_User, time=0, value=0}, {tangentMode=RCTM_User, time=1, value=1}]}}}
      CurveIndex=$System.Size Lerp
     )
     Y=&FloatFromCurve(
      FloatCurve=DataInterface() {propertyValues {Curve {keys [{tangentMode=RCTM_User, time=0, value=0}, {tangentMode=RCTM_User, time=1, value=1}]}}}
      CurveIndex=$System.Size Lerp
     )
     Z=1
    )) {module=ScaleMeshSize001}
   &ScaleMeshSize(Scale Factor=&VectorFromCurve(
     VectorCurve=DataInterface() {propertyValues {
       XCurve {keys [
         {arriveTangent=5.0383, interpMode=RCIM_Cubic, leaveTangent=5.0383, tangentMode=RCTM_User, time=0, value=0.47}
         {arriveTangent=1.03, interpMode=RCIM_Cubic, leaveTangent=1.03, time=0.7, value=2.5}
         <at 1 1.5 RCIM_Cubic>
        ]}
       YCurve {keys [{arriveTangent=6.1645, interpMode=RCIM_Cubic, leaveTangent=6.1645, tangentMode=RCTM_User, time=0.0035, value=0.4903}, <at 0.9786 1.29>]}
       ZCurve {keys [
         <at -0.0011 1.0067 RCIM_Cubic>
         {arriveTangent=-1.635, interpMode=RCIM_Cubic, leaveTangent=-1.635, tangentMode=RCTM_User, time=1.0016, value=0.28}
        ]}
      }}
     Scale Curve=&VectorFromFloat(Value=1)
    )) {module=ScaleMeshSize002}
   &ScaleColor(
    Scale Alpha=&FloatFromCurve(FloatCurve=DataInterface() {propertyValues {Curve {keys [<at 0 1>,<at 1 0>]}}})
    Scale RGB=&VectorFromFloat(Value=&FloatFromCurve(
      FloatCurve=DataInterface() {propertyValues {Curve {keys [<at 0 1 RCIM_Cubic>, {arriveTangent=-1.4981, interpMode=RCIM_Cubic, leaveTangent=-1.4981, tangentMode=RCTM_User, time=1, value=0}]}}}
      Scale Curve=15
     ))
    ScaleA=true
    ScaleRGB=true
   )
   &ScaleColor(
    Scale RGB=&VectorFromFloat(Value=&FloatFromCurve(FloatCurve=DataInterface() {propertyValues {Curve {keys [<at 0 1>,<at 1 0>]}}}))
    ScaleA=true
    ScaleRGB=true
   ) {module=ScaleColor001}
   &UpdateMeshOrientation(
    Facing Direction=&Subtract_Vector(
     A=Linked() {linkedVariable {name=Particles.Position, type=NiagaraPosition}}
     B=&Add_Vector(
      A=Linked() {linkedVariable {name=Output.SkeletalMeshLocation.SampledPositionWithoutOffset, type=NiagaraPosition}}
      B=&Multiply_Vector(
       A=&Multiply_VectorByFloat(
        Vector=&ReturnMeshOrientationAxis(Mesh Orientation Axis=::XAxis, Mesh Orientation Quat=$Particles.SkeletalMeshLocation.SampledRotation)
        Float=30
       )
       B=<1 1 1>
      )
     )
    )
    Facing Interpolation=false
    Orientation Method=::NewEnumerator1
    Override Quat=false
    TransformFacing=false
    TransformUp=false
    Up Direction:Vector3f=&ReturnMeshOrientationAxis(Mesh Orientation Axis=::XAxis, Mesh Orientation Quat=$Particles.SkeletalMeshLocation.SampledRotation)
   )
  ]
  rendererClasses [&NiagaraMeshRendererProperties]
  renderers [{rendererClass=&NiagaraMeshRendererProperties, rendererIndex=0}]
 }
 Trail001 {
  emitterUpdateScript [
   &EmitterState(
    Enable Visibility Culling=true
    Scalability Mode=::NewEnumerator1
    Visibility Cull Delay=0.01
    Visibility Cull Response:ENiagaraExecutionStateManagement=::SleepAndClearParticles
   ) {module=EmitterState001}
   &self:CalcDistance_Travelled(Velocity Vector:Vector3f=&Multiply_Vector(A=$Engine.Owner.Velocity, B=&VectorFromFloat(Value=-1))) {module=CalcDistance_Travelled}
   &SpawnRate(SpawnRate=120, Use Spawn Probability=false)
  ]
  particleSpawnScript [
   &InitializeRibbon(
    Color=&MakeCustomLinearColorFromBool(Bool=$System.IsAltBool, True Color=<rgba 0.3001 0.3001 1 1.5>, False Color=<rgba 1.7867 0.4 2 1>)
    Lifetime=0.15
    UsePositionOffset=false
    Write Material Random=true
   )
   /SetVariables(Particles.RibbonUVDistance:NiagaraFloat=$Emitter.DistanceTraveled) {isSetParameters=true}
   /SetVariables(Particles.RibbonWidth=25) {isSetParameters=true}
   &SkeletalMeshLocation(
    Bone / Socket Index=3
    Bone Sampling Mode=::NewEnumerator4
    Interpolate Mesh Sampling=false
    MaskSpawnGroup=false
    Sampled Position Offset=<30 0 0>
    Skeletal Mesh=$User.SkeletalMesh
   )
   /SetVariables(Particles.RibbonFacing=&ReturnMeshOrientationAxis(Mesh Orientation Axis=::XAxis, Mesh Orientation Quat=$Particles.SkeletalMeshLocation.SampledRotation)) {isSetParameters=true}
   /SetVariables(Particles.pink line:NiagaraFloat=&Multiply_Float(A=$Particles.RibbonWidth, B=0.75)) {isSetParameters=true}
   /SetVariables(Particles.SpriteSize=<5 5>) {isSetParameters=true}
  ]
  particleUpdateScript [
   &ParticleState()
   &ScaleColor(
    Scale Alpha=&FloatFromCurve(
     FloatCurve=DataInterface() {propertyValues {Curve {keys [
         {
          arriveTangent=-8.5427
          arriveTangentWeight=0.8208
          interpMode=RCIM_Cubic
          leaveTangent=-8.5427
          leaveTangentWeight=0.8208
          tangentMode=RCTM_User
          tangentWeightMode=RCTWM_WeightedBoth
          time=0.1
          value=1
         }
         {
          arriveTangent=-0.9101
          arriveTangentWeight=0.5943
          interpMode=RCIM_Cubic
          leaveTangent=-0.3338
          leaveTangentWeight=0.2826
          tangentMode=RCTM_Break
          tangentWeightMode=RCTWM_WeightedBoth
          time=0.6
          value=0
         }
        ]}}}
     Scale Curve=2
    )
    ScaleA=true
    ScaleRGB=false
   )
   /SetVariables(Particles.Line Color:LinearColor=&MultiplyLinearColor(
     Linear Color=&MakeCustomLinearColorFromBool(Bool=$System.IsAltBool, True Color=<rgba 0.25 0.375 1 0.2>, False Color=<rgba 0.9 0.25 1 0.2>)
     Scale Factor=&MakeFloat4FromVectorAndFloat(
      "Vector (RGB)"=<1 1 1>
      "Float (Alpha)"=&FloatFromCurve(FloatCurve=DataInterface() {propertyValues {Curve {keys [<at 0 1>,<at 0.317 1>,<at 0.367 0>]}}})
     )
    )) {isSetParameters=true}
  ]
  rendererClasses [&NiagaraRibbonRendererProperties]
  renderers [{rendererClass=&NiagaraRibbonRendererProperties}]
 }
}
summary {
 systemName=NS_ForsakeBuiltin_Swing
 emitters [
  {emitterName=Sprites, rendererClasses [&NiagaraSpriteRendererProperties]}
  {emitterName=Trail001, rendererClasses [&NiagaraRibbonRendererProperties]}
  {emitterName=Lights, rendererClasses [&NiagaraSpriteRendererProperties, &NiagaraLightRendererProperties]}
  {emitterName=Thrust001, rendererClasses [&NiagaraMeshRendererProperties]}
 ]
 userVariables [
  User.IsAlt {type=NiagaraFloat, defaultValue=0, description=""}
  User.SkeletalMesh {
   type=NiagaraDataInterfaceSkeletalMesh
   defaultValue {
    struct=NiagaraExt_VariableValue_DataInterface
    value {dataInterface=&self:NiagaraDataInterfaceSkeletalMesh_0}
   }
  }
 ]
}
